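# Audio level sequence for the GrillMoon Countdown tdm0500 sequence, timing music, ambience and a master audio section.
# Event trigger tracks on the music and ambience components, both parented to the Day5 SequenceRunner binding, fire two music events and one ambience event.
# A root audio track named "tdm" plays one section at default volume and pitch, starting shortly after the first music event.
# The mask timer Blueprint is bound but has no tracks.
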LevelSequence /GrillMoon_Countdown2/Sequence/tdm/tdm0500/GrillMoon_Countdown_tdm0500_aud
bindings [
 GrillMoon_Countdown_BPmaskTimer {id=#1, possessed_class=GrillMoon_Countdown_BPmaskTimer_C, tracks []}
 BP_GrillMoon_POI_MusicComponent {
  id=#2
  parent_id=#3
  possessed_class=BP_GrillMoon_POI_MusicComponent_C
  tracks [{
    type=MovieSceneEventTrack
    sections [{
      type=MovieSceneEventTriggerSection
      channels [None {
        type=MovieSceneScriptingEventChannel
        keys [[3624.095, "<Struct 'MovieSceneEvent' (0x850243160e0) {}>"], [4024.555, "<Struct 'MovieSceneEvent' (0x850243112c0) {}>"]]
       }]
     }]
   }]
 }
 BP_GrillMoon_POI_AmbienceComponent {
  id=#4
  parent_id=#3
  possessed_class=BP_GrillMoon_POI_AmbienceComponent_C
  tracks [{
    type=MovieSceneEventTrack
    sections [{
      type=MovieSceneEventTriggerSection
      channels [None {type=MovieSceneScriptingEventChannel, const ["<Struct 'MovieSceneEvent' (0x85024311d60) {}>", 1]}]
     }]
   }]
 }
 BP_GrillMoon_Day5_SequenceRunner {id=#3, possessed_class=null, tracks []}
]
display_rate {denominator=1, numerator=60}
tick_resolution {denominator=1, numerator=24000}
tracks [{
  type=MovieSceneAudioTrack
  display_name=tdm
  sections [{
    type=MovieSceneAudioSection
    end_frame=4197
    start_frame=3640
    channels [
     Volume {type=MovieSceneScriptingFloatChannel, default=1}
     Pitch {type=MovieSceneScriptingFloatChannel, default=1}
     AttachActor {type=MovieSceneScriptingActorReferenceChannel}
    ]
   }]
 }]
