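# Level Sequence for the tdm0500 "tec" layer of GrillMoon Countdown, driving mask timer effects and material state.
# Particle tracks activate five Masks Niagara components and a PortalContact component, all under the GrillMoon_Countdown_BPmaskTimer binding.
# An MPC_GrillMoon track keys Mask_Phase from 4 to 5, advancing the shared material phase.
# An M_RicePaper_Simple binding carries a spawn track and an identity transform.
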LevelSequence /GrillMoon_Countdown2/Sequence/tdm/tdm0500/GrillMoon_Countdown_tdm0500_tec
bindings [
 NS_GrillMoon_Countdown_Masks_003 {
  id=#1
  parent_id=#2
  possessed_class=NiagaraComponent
  tracks [{
    type=MovieSceneParticleTrack
    display_name=FX System
    sections [{type=MovieSceneParticleSection, channels [None {type=MovieSceneScriptingParticleChannel, const ["<ParticleKey.ACTIVATE: 0>", 1]}]}]
   }]
 }
 GrillMoon_Countdown_BPmaskTimer {id=#2, possessed_class=GrillMoon_Countdown_BPmaskTimer_C, tracks []}
 NS_GrillMoon_Countdown_Masks_001 {
  id=#3
  parent_id=#2
  possessed_class=NiagaraComponent
  tracks [{
    type=MovieSceneParticleTrack
    display_name=FX System
    sections [{type=MovieSceneParticleSection, channels [None {type=MovieSceneScriptingParticleChannel, const ["<ParticleKey.ACTIVATE: 0>", 1]}]}]
   }]
 }
 NS_GrillMoon_Countdown_Masks_005 {
  id=#4
  parent_id=#2
  possessed_class=NiagaraComponent
  tracks [{
    type=MovieSceneParticleTrack
    display_name=FX System
    sections [{type=MovieSceneParticleSection, channels [None {type=MovieSceneScriptingParticleChannel, const ["<ParticleKey.ACTIVATE: 0>", 1]}]}]
   }]
 }
 M_RicePaper_Simple {
  id=#5
  possessed_class=Object
  tracks [
   {type=MovieSceneSpawnTrack, sections [{type=MovieSceneSpawnSection, channels [None {type=MovieSceneScriptingBoolChannel, default=true}]}]}
   {
    type=MovieScene3DTransformTrack
    sections [{
      type=MovieScene3DTransformSection
      channels [
       Location.X {type=MovieSceneScriptingDoubleChannel, default=0}
       Location.Y {type=MovieSceneScriptingDoubleChannel, default=0}
       Location.Z {type=MovieSceneScriptingDoubleChannel, default=0}
       Rotation.X {type=MovieSceneScriptingDoubleChannel, default=0}
       Rotation.Y {type=MovieSceneScriptingDoubleChannel, default=0}
       Rotation.Z {type=MovieSceneScriptingDoubleChannel, default=0}
       Scale.X {type=MovieSceneScriptingDoubleChannel, default=1}
       Scale.Y {type=MovieSceneScriptingDoubleChannel, default=1}
       Scale.Z {type=MovieSceneScriptingDoubleChannel, default=1}
      ]
     }]
   }
  ]
 }
 NS_GrillMoon_Countdown_Masks_002 {
  id=#6
  parent_id=#2
  possessed_class=NiagaraComponent
  tracks [{
    type=MovieSceneParticleTrack
    display_name=FX System
    sections [{type=MovieSceneParticleSection, channels [None {type=MovieSceneScriptingParticleChannel, const ["<ParticleKey.ACTIVATE: 0>", 1]}]}]
   }]
 }
 NS_GrillMoon_Countdown_PortalContact {
  id=#7
  parent_id=#2
  possessed_class=NiagaraComponent
  tracks [{
    type=MovieSceneParticleTrack
    display_name=FX System
    sections [{type=MovieSceneParticleSection, channels [None {type=MovieSceneScriptingParticleChannel, const ["<ParticleKey.ACTIVATE: 0>", 1]}]}]
   }]
 }
 NS_GrillMoon_Countdown_Masks_004 {
  id=#8
  parent_id=#2
  possessed_class=NiagaraComponent
  tracks [{
    type=MovieSceneParticleTrack
    display_name=FX System
    sections [{type=MovieSceneParticleSection, channels [None {type=MovieSceneScriptingParticleChannel, const ["<ParticleKey.ACTIVATE: 0>", 1]}]}]
   }]
 }
]
display_rate {denominator=1, numerator=60}
tick_resolution {denominator=1, numerator=24000}
tracks [{
  type=MovieSceneMaterialParameterCollectionTrack
  display_name=MPC_GrillMoon
  sections [{type=MovieSceneParameterSection, channels [Mask_Phase {type=MovieSceneScriptingFloatChannel, default=5, keys [[3904,4],[4006,5]]}]}]
 }]
